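# Sound cue named for lightning at the Deimos monster spawner portal, though its node graph is empty and has no root, so it plays no audio as it stands.
# It retains playback setup: the enemy spawn sound class, a reduced volume and Husk spawner attenuation.
# It references the shared Deimos spawner concurrency set and also carries override values for attenuation and for a small, distance-prioritized voice limit.
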SoundCue /Game/Sounds/Fort_Monster_Deimos/Spawner/DeimosPortal_Lightning_Cue
root=null
nodes []
properties {
 attenuationSettings=/Game/Sounds/Attenuation/Fort_Monster_Husk_Spawn_Att
 maxDistance=18000
 soundClassObject=/Game/Sounds/FortSoundClassesAndModes/Fort_Enemy_Spawn
 volumeMultiplier=0.6
 attenuationOverrides {falloffDistance=4000, attenuationShapeExtents {x=1000}}
 concurrencyOverrides {maxCount=2, resolutionRule=StopFarthestThenPreventNew}
 concurrencySet [/Game/Sounds/Concurrency_Rules/Deimos_Spawner_Global_Concurrency]
}
