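# Sound cue for the Deimos monster spawner creation effect, a Creative wide-range variant built on distant spawn recordings.
# A Random node picks one of six equally weighted waves, and a Modulator varies it slightly, with pitch capped at normal.
# Wide attenuation shapes falloff on the in-game mission sound class, with boosted volume.
# A concurrency override caps simultaneous plays at four, rejecting new ones and scaling volume among overlapping instances.
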SoundCue /Game/Sounds/Fort_Monster_Deimos/Spawner/DeimosSpawnerCreate_Creative_Wide_Cue
root=Modulator_4
nodes [
 {id=Modulator_4, children [Random_6], properties {pitchMax=1}}
 {
  id=Random_6
  children [WavePlayer_28, WavePlayer_29, WavePlayer_30, WavePlayer_31, WavePlayer_32, WavePlayer_33]
  properties {weights [1,1,1,1,1,1]}
 }
 {id=WavePlayer_28, properties {soundWaveAssetPtr=&/Deimos_SpawnerSpawn_Distant_01}}
 {id=WavePlayer_29, properties {soundWaveAssetPtr=&/Deimos_SpawnerSpawn_Distant_02}}
 {id=WavePlayer_30, properties {soundWaveAssetPtr=&/Deimos_SpawnerSpawn_Distant_03}}
 {id=WavePlayer_31, properties {soundWaveAssetPtr=&/Deimos_SpawnerSpawn_Distant_04}}
 {id=WavePlayer_32, properties {soundWaveAssetPtr=&/Deimos_SpawnerSpawn_Distant_05}}
 {id=WavePlayer_33, properties {soundWaveAssetPtr=&/Deimos_SpawnerSpawn_Distant_06}}
]
properties {
 attenuationSettings=&/Fort_Deimos_SpawnerCreative_Wide_Att
 bOverrideConcurrency=true
 maxDistance=50000
 soundClassObject=/Game/Sounds/FortSoundClassesAndModes/Fort_Mission_IngameSounds
 volumeMultiplier=1.5
 concurrencyOverrides {maxCount=4, resolutionRule=PreventNew, volumeScale=0.5}
}
