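# Sound cue for a Deimos monster spawner being created, routed to the in-game mission sound class with boosted volume.
# A Random node picks one of two equally weighted spawn waves and feeds an Attenuation node that uses the large spawner ambient attenuation asset.
# Overridden concurrency caps simultaneous instances, stopping the farthest before preventing new ones.
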SoundCue /Game/Sounds/Fort_Monster_Deimos/Spawner/DeimosSpawnerCreate_Cue
root=Attenuation_0
nodes [
 {
  id=Attenuation_0
  children [Random_0]
  properties {attenuationSettings=&/Fort_Deimos_SpawnerAmbientLarge_Att}
 }
 {id=Random_0, children [WavePlayer_3,WavePlayer_4], properties {weights [1,1]}}
 {id=WavePlayer_3, properties {soundWaveAssetPtr=&/Deimos_SpawnerSpawn_01}}
 {id=WavePlayer_4, properties {soundWaveAssetPtr=&/Deimos_SpawnerSpawn_02}}
]
properties {
 attenuationSettings=&Fort_Deimos_SpawnerAmbientLarge_Att
 bOverrideConcurrency=true
 maxDistance=4000
 soundClassObject=/Game/Sounds/FortSoundClassesAndModes/Fort_Mission_IngameSounds
 volumeMultiplier=2
 attenuationOverrides {falloffDistance=1000000}
 concurrencyOverrides {maxCount=4, resolutionRule=StopFarthestThenPreventNew}
}
