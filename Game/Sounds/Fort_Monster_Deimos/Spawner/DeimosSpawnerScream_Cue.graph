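# Sound cue for the Deimos monster spawner's scream, layered from borrowed Storm King, Cube and sword impact waves.
# A mixer plays three randomized layers together: large Cube impacts, Storm King spawner sounds, and much quieter sword jump-attack impacts.
# Both impact layers are pitched slightly down through modulators.
# It routes to the boss-fight player-critical sound class, with long-range attenuation and a low concurrency cap.
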SoundCue /Game/Sounds/Fort_Monster_Deimos/Spawner/DeimosSpawnerScream_Cue
root=Mixer_0
nodes [
 {id=Mixer_0, children [Modulator_1,Random_1,Modulator_0], properties {inputVolume [1,1,0.2]}}
 {id=Modulator_0, children [Random_2], properties {pitchMax=0.95, pitchMin=0.93}}
 {id=Modulator_1, children [Random_0], properties {pitchMax=0.95, pitchMin=0.93}}
 {id=Random_0, children [WavePlayer_1,WavePlayer_2], properties {weights [1,1]}}
 {id=Random_1, children [WavePlayer_4,WavePlayer_5,WavePlayer_6], properties {weights [1,1,1]}}
 {id=Random_2, children [WavePlayer_7,WavePlayer_8,WavePlayer_9], properties {weights [1,1,1]}}
 {id=WavePlayer_1, properties {soundWaveAssetPtr=/Game/Athena/Sounds/Events/Cube/Cube_ImpactLarge_01}}
 {id=WavePlayer_2, properties {soundWaveAssetPtr=/Game/Athena/Sounds/Events/Cube/Cube_ImpactLarge_03}}
 {id=WavePlayer_4, properties {soundWaveAssetPtr=/Game/Sounds/Fort_Monster_Storm_King/StormKing_Deimos_Spawners_10}}
 {id=WavePlayer_5, properties {soundWaveAssetPtr=/Game/Sounds/Fort_Monster_Storm_King/StormKing_Deimos_Spawners_11}}
 {id=WavePlayer_6, properties {soundWaveAssetPtr=/Game/Sounds/Fort_Monster_Storm_King/StormKing_Deimos_Spawners_12}}
 {id=WavePlayer_7, properties {soundWaveAssetPtr=/Game/Sounds/Fort_GamePlay_Sounds/BRSword/AB_Jump_Attack_Impact_02}}
 {id=WavePlayer_8, properties {soundWaveAssetPtr=/Game/Sounds/Fort_GamePlay_Sounds/BRSword/AB_Jump_Attack_Impact_03}}
 {id=WavePlayer_9, properties {soundWaveAssetPtr=/Game/Sounds/Fort_GamePlay_Sounds/BRSword/AB_Jump_Attack_Impact_01}}
]
properties {
 bOverrideAttenuation=true
 bOverrideConcurrency=true
 maxDistance=27000
 soundClassObject=/Game/Sounds/FortSoundClassesAndModes/Fort_BossFight_SoundClass/Fort_BossFight_PlayerCritical
 volumeMultiplier=1.5
 attenuationOverrides {falloffDistance=2000, stereoSpread=500, attenuationShapeExtents {x=25000}}
 concurrencyOverrides {maxCount=2}
}
