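# Sound cue for the charge-up tell of a Deimos monster spawner, in the enemy spawn sound class.
# A Random node picks one of two equally weighted SpawnerCharge waves and feeds it through a Modulator at default settings.
# Playback uses the large Deimos spawner ambient attenuation, and a concurrency override caps simultaneous instances, refusing new ones at the limit.
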SoundCue /Game/Sounds/Fort_Monster_Deimos/Spawner/DeimosSpawnerTell_Cue
root=Modulator_2
nodes [
 {id=Modulator_2, children [Random_0]}
 {id=Random_0, children [WavePlayer_1,WavePlayer_4], properties {bShouldExcludeFromBranchCulling=true, weights [1,1]}}
 {id=WavePlayer_1, properties {soundWaveAssetPtr=&/Deimos_SpawnerCharge_01}}
 {id=WavePlayer_4, properties {soundWaveAssetPtr=&/Deimos_SpawnerCharge_02}}
]
properties {
 attenuationSettings=&/Fort_Deimos_SpawnerAmbientLarge_Att
 bOverrideConcurrency=true
 maxDistance=4000
 soundClassObject=/Game/Sounds/FortSoundClassesAndModes/Fort_Enemy_Spawn
 volumeMultiplier=1
 concurrencyOverrides {maxCount=4, resolutionRule=PreventNew}
}
