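# Sound cue for the Deimos monster spawner explosion, played in the medium explosives sound class with concurrency limits.
# A mixer layers three randomized wave sets: an immediate, slightly boosted rock layer, a briefly delayed close explosion, and a later distant explosion.
# Each layer passes through a modulator, with wider pitch range on the rock, and its own attenuation node, using distant settings for the distant layer and close settings for the others.
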SoundCue /Game/Sounds/Fort_Monster_Deimos/Spawner/DeimosSpawner_Explode_Cue
root=Mixer_0
nodes [
 {
  id=Attenuation_0
  children [Modulator_0]
  properties {attenuationSettings=&/Fort_Deimos_SpawnerExplode_Close_Att}
 }
 {
  id=Attenuation_1
  children [Modulator_1]
  properties {attenuationSettings=&/Fort_Deimos_SpawnerExplode_Distant_Att}
 }
 {
  id=Attenuation_2
  children [Modulator_2]
  properties {attenuationSettings=&Fort_Deimos_SpawnerExplode_Close_Att}
 }
 {id=Delay_0, children [Attenuation_1], properties {delayMax=0.5, delayMin=0.5}}
 {id=Delay_1, children [Attenuation_0], properties {delayMax=0.4, delayMin=0.25}}
 {id=Mixer_0, children [Attenuation_2, Delay_1, Delay_0], properties {inputVolume [1.25,1,1]}}
 {id=Modulator_0, children [Random_0], properties {volumeMax=1, volumeMin=1}}
 {id=Modulator_1, children [Random_1], properties {volumeMax=1, volumeMin=1}}
 {id=Modulator_2, children [Random_2], properties {pitchMax=1.2, volumeMax=1, volumeMin=1}}
 {
  id=Random_0
  children [WavePlayer_2,WavePlayer_3,WavePlayer_4,WavePlayer_5]
  properties {bShouldExcludeFromBranchCulling=true, weights [1,1,1,1]}
 }
 {
  id=Random_1
  children [WavePlayer_6,WavePlayer_7,WavePlayer_8,WavePlayer_9]
  properties {bShouldExcludeFromBranchCulling=true, weights [1,1,1,1]}
 }
 {id=Random_2, children [WavePlayer_1, WavePlayer_10, WavePlayer_11], properties {bShouldExcludeFromBranchCulling=true, weights [1,1,1]}}
 {id=WavePlayer_1, properties {soundWaveAssetPtr=&/DeimosSpawner_Explode_Close_Rock_01}}
 {id=WavePlayer_10, properties {soundWaveAssetPtr=&/DeimosSpawner_Explode_Close_Rock_02}}
 {id=WavePlayer_11, properties {soundWaveAssetPtr=&/DeimosSpawner_Explode_Close_Rock_03}}
 {id=WavePlayer_2, properties {soundWaveAssetPtr=&/DeimosSpawner_Explode_Close_01}}
 {id=WavePlayer_3, properties {soundWaveAssetPtr=&/DeimosSpawner_Explode_Close_02}}
 {id=WavePlayer_4, properties {soundWaveAssetPtr=&/DeimosSpawner_Explode_Close_03}}
 {id=WavePlayer_5, properties {soundWaveAssetPtr=&/DeimosSpawner_Explode_Close_04}}
 {id=WavePlayer_6, properties {soundWaveAssetPtr=&/DeimosSpawner_Explode_Distant_01}}
 {id=WavePlayer_7, properties {soundWaveAssetPtr=&/DeimosSpawner_Explode_Distant_02}}
 {id=WavePlayer_8, properties {soundWaveAssetPtr=&/DeimosSpawner_Explode_Distant_03}}
 {id=WavePlayer_9, properties {soundWaveAssetPtr=&/DeimosSpawner_Explode_Distant_04}}
]
properties {
 attenuationSettings=&Fort_Deimos_SpawnerExplode_Close_Att
 maxDistance=18000
 soundClassObject=/Game/Sounds/FortSoundClassesAndModes/Fort_Explosives_Medium
 virtualizationMode=Disabled
 volumeMultiplier=0.1
 concurrencyOverrides {bVolumeScaleCanRelease=true, maxCount=2, volumeScale=0.15, volumeScaleAttackTime=0.05, volumeScaleReleaseTime=1.5}
 concurrencySet [
  &/DeimosSpawner_Explo_Global_Conc
  &/DeimosSpawner_Explo_Owner_Conc
 ]
}
