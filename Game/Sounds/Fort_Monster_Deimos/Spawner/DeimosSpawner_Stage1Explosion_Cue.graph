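# Sound cue named for the stage 1 explosion of the Deimos monster spawner, with no nodes and a null root, so it has no audio source wired in.
# It carries playback settings only: large spawner ambient attenuation, the enemy spawn sound class, boosted volume and disabled virtualization.
# It references a global spawner explosion concurrency set, alongside a single-instance concurrency override.
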SoundCue /Game/Sounds/Fort_Monster_Deimos/Spawner/DeimosSpawner_Stage1Explosion_Cue
root=null
nodes []
properties {
 attenuationSettings=&/Fort_Deimos_SpawnerAmbientLarge_Att
 maxDistance=4000
 soundClassObject=/Game/Sounds/FortSoundClassesAndModes/Fort_Enemy_Spawn
 virtualizationMode=Disabled
 volumeMultiplier=1.25
 concurrencyOverrides {maxCount=1}
 concurrencySet [&/DeimosSpawner_Explo_Global_Conc]
}
