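# Sound cue named for the stage 3 buildup of the Deimos monster spawner, but it has no nodes or root, so it plays no audio as it stands.
# It keeps only playback settings: the large ambient spawner attenuation, a third-person ranged weapon fire sound class, and a concurrency rule named for one instance per owner, stopping the oldest.
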SoundCue /Game/Sounds/Fort_Monster_Deimos/Spawner/DeimosSpawner_Stage3Buildup_Cue
root=null
nodes []
properties {
 attenuationSettings=&/Fort_Deimos_SpawnerAmbientLarge_Att
 maxDistance=4000
 soundClassObject=&Fort_WeaponRanged_Fire_3P_PSM
 volumeMultiplier=1
 concurrencySet [/Game/Sounds/Concurrency_Rules/OneMaximum_Owner_StopOldest]
}
