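# Looping ambient sound cue for the Deimos monster spawner, routed to the enemy spawn sound class.
# A mixer layers a small-focus attenuated loop, switched among three variants by the integer Damage parameter, with a large-attenuation tone bed.
# A quality-level node picks the 60-second or 15-second tone loop for that bed.
# A concurrency override limits simultaneous instances, stopping the quietest.
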SoundCue /Game/Sounds/Fort_Monster_Deimos/Spawner/Deimos_Spawner_Loop1_Cue
root=Mixer_2
nodes [
 {
  id=Attenuation_2
  children [Switch_0]
  properties {attenuationSettings=&/Fort_Deimos_SpawnerAmbientSmallFocus_Att}
 }
 {
  id=Attenuation_5
  children [QualityLevel_0]
  properties {attenuationSettings=&/Fort_Deimos_SpawnerAmbientLarge_Att}
 }
 {id=Mixer_2, children [Attenuation_2, Attenuation_5], properties {inputVolume [1,1]}}
 {id=QualityLevel_0, children [WavePlayer_6,WavePlayer_6,WavePlayer_7,WavePlayer_7]}
 {id=Switch_0, children [WavePlayer_2,WavePlayer_2,WavePlayer_2,WavePlayer_0,WavePlayer_1], properties {intParameterName=Damage}}
 {id=WavePlayer_0, properties {bLooping=true, soundWaveAssetPtr=&/Deimos_Spawner_Loop2}}
 {id=WavePlayer_1, properties {bLooping=true, soundWaveAssetPtr=&/Deimos_Spawner_Loop3}}
 {id=WavePlayer_2, properties {bLooping=true, soundWaveAssetPtr=&/Deimos_Spawner_Loop1}}
 {id=WavePlayer_6, properties {bLooping=true, soundWaveAssetPtr=&/deimos_spawner_tone60sLoop}}
 {id=WavePlayer_7, properties {bLooping=true, soundWaveAssetPtr=&/deimos_spawner_tone15sLoop}}
]
properties {
 bOverrideConcurrency=true
 maxDistance=4000
 soundClassObject=/Game/Sounds/FortSoundClassesAndModes/Fort_Enemy_Spawn
 volumeMultiplier=0.6
 concurrencyOverrides {maxCount=3, resolutionRule=StopQuietest}
}
